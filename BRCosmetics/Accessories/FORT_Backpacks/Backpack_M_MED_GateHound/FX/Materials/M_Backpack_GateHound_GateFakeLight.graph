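# Translucent unlit Niagara material that fakes gate light for the GateHound backpack effect.
# A vertical two-color gradient, blended to an alternate palette by the Variant dynamic parameter, feeds an emissive adjustment function.
# Opacity math projects the scene position behind the translucency, relative to the particle, onto a local axis scaled by the push distance.
# That falloff is shaped by the gradient's fourth channel and multiplied by an offset UV bottom fade.
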Material /BRCosmetics/Accessories/FORT_Backpacks/Backpack_M_MED_GateHound/FX/Materials/M_Backpack_GateHound_GateFakeLight
graph {
 properties {
  bUsedWithNiagaraSprites=true
  blendMode=BLEND_Translucent
  pixelDepthOffsetMode=PDOM_Legacy
  shadingModel=MSM_Unlit
  translucencyPass=MTP_BeforeDOF
  usedShadingModels=Unlit
 }
 nodes [
  n1 = VectorParameter() {parameterName=Color Top, defaultValue=<rgba 0.03 0.06 0.06 3>}
  n2 = VectorParameter() {parameterName=Color Bottom, defaultValue=<rgba 0 0.4 0.2667 2>}
  n3 = ComponentMask(None=TextureCoordinate()) {g=true}
  n4 = VectorParameter() {parameterName=Variant Color Top, defaultValue=<rgba 0.03 0.035 0.06 3>}
  n5 = VectorParameter() {parameterName=Variant Color Bottom, defaultValue=<rgba 0 0.12 0.4 2>}
  n6 = LinearInterpolate(
   A=LinearInterpolate(A=AppendVector(A=@n1.RGB, B=@n1.RGB), B=AppendVector(A=@n2.RGB, B=@n2.RGB), Alpha=@n3)
   B=LinearInterpolate(A=AppendVector(A=@n4.RGB, B=@n4.RGB), B=AppendVector(A=@n5.RGB, B=@n5.RGB), Alpha=@n3)
   Alpha=DynamicParameter() {defaultValue=<rgba 0 0 0 0>, paramNames [Reactivity Alpha, Reactivity Timeline, Variant, Param4]}.Variant
  )
  MaterialFunctionCall(Emissive=ComponentMask(None=@n6) {b=true, g=true, r=true}) {materialFunction=&MF_FX_EmissiveAdjustments}
  n7 = Transform(None=Constant3Vector() {constant=<rgba 0 -1 0 0>}) {transformSourceType=TRANSFORMSOURCE_Local}
  n8 = ScalarParameter() {defaultValue=2, parameterName=Vertex Normal Push}
  Multiply(A=@n7, B=@n8)
  n9 = DotProduct(
   A=Subtract(
    A=MaterialFunctionCall() {materialFunction=&WorldPositionBehindTranslucency}.Result
    B=ParticlePositionWS()
   )
   B=@n7
  )
  n10 = OneMinus(None=Saturate(None=Divide(A=@n9, B=@n8)))
  n11 = TextureCoordinate()
  n12 = Subtract(A=OneMinus(None=ComponentMask(None=@n11) {g=true}), B=ScalarParameter() {defaultValue=0.15, parameterName=Opacity Bottom Fade Offset})
  n13 = ScalarParameter() {defaultValue=0.5, parameterName=Opacity Bottom Fade Clip Size}
  Multiply(A=Power(Base=Frac(None=@n10), Exp=ComponentMask(None=@n6) {a=true}), B=Saturate(None=Divide(A=@n12, B=@n13)))
 ]
}
