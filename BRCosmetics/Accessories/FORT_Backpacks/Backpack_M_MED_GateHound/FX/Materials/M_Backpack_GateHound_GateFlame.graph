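# Translucent unlit Niagara material for the GateHound backpack's gate flame effect.
# Three-way panned plasma blends with a center gradient and panning lightning texture, using skewed UVs and water-normal distortion.
# The result drives two four-color ramps, with Niagara dynamic parameters selecting the variant and supplying reactivity, including a timeline-driven gradient pulse.
# Platform switches drop distortion and depth fade on lower-end targets.
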Material /BRCosmetics/Accessories/FORT_Backpacks/Backpack_M_MED_GateHound/FX/Materials/M_Backpack_GateHound_GateFlame
graph {
 properties {
  bUsedWithNiagaraSprites=true
  blendMode=BLEND_Translucent
  pixelDepthOffsetMode=PDOM_Legacy
  shadingModel=MSM_Unlit
  translucencyPass=MTP_BeforeDOF
  twoSided=true
  usedShadingModels=Unlit
 }
 nodes [
  Add(A=NamedRerouteUsage(), B=NamedRerouteUsage())
  n1 = NamedRerouteUsage()
  n2 = AppendVector(
   A=ScalarParameter() {defaultValue=0.5, parameterName=Plasma UV Scale U}
   B=ScalarParameter() {defaultValue=1, parameterName=Plasma UIV Scale V}
  )
  n3 = MaterialFunctionCall!(
   UV=Add(A=Multiply(A=@n1, B=@n2), B=NamedRerouteUsage())
   Rotation=ScalarParameter() {defaultValue=0.25, parameterName=Plasma Panner Rotation}
   Speed=ScalarParameter() {defaultValue=0.4, parameterName=Plasma Panner Speed}
   Time=NamedRerouteUsage()
   Rotate UVs
   Fractional Part
  ) {materialFunction=/Game/Accessories/FORT_Backpacks/Backpack_M_MED_PitGlass/FX/Materials/MF_3UVPanner}
  n4 = TextureObject() {samplerType=SAMPLERTYPE_LinearGrayscale, texture=/BRCosmetics/Effects/Fort_Effects/Emotes/UltraLight/Texture/T_Plasma_D_R_Greyscale}
  n5 = TextureSample(UVs=@n3.UV1, Tex=@n4) {samplerType=SAMPLERTYPE_LinearColor}
  n6 = TextureSample(UVs=@n3.UV2, Tex=@n4) {samplerType=SAMPLERTYPE_LinearColor}
  n7 = Divide(A=Add(A=Add(A=@n5.RGB, B=@n6.RGB), B=TextureSample(UVs=@n3.UV3, Tex=@n4) {samplerType=SAMPLERTYPE_LinearColor}.RGB)) {constB=3}
  n8 = ComponentMask(None=@n7) {r=true}
  n9 = NamedRerouteUsage()
  n10 = ScalarParameter() {defaultValue=1, parameterName=Lightning Blend Intensity}
  n11 = Power(Base=NamedRerouteUsage(), Exp=ScalarParameter() {defaultValue=1, parameterName=Center Gradient Exponent})
  n12 = LinearInterpolate(
   A=Multiply(A=@n8) {constB=0.5}
   B=MaterialFunctionCall!(
    Blend=Multiply(A=@n9, B=@n10) {constB=0.5}
    Base=MaterialFunctionCall(Base=@n11, Blend=@n8) {materialFunction=&Blend_Overlay}.Result
   ) {materialFunction=/Engine/Functions/Engine_MaterialFunctions03/Blends/Blend_Screen}.Result
   Alpha=NamedRerouteUsage()
  )
  n13 = ScalarParameter() {defaultValue=-0.333, parameterName=Blended Texture Exponent Top}
  n14 = Multiply(A=ScalarParameter() {defaultValue=1, parameterName=Blended Texture Exponent Bottom}, B=NamedRerouteUsage())
  n15 = Saturate(None=ComponentMask(None=NamedRerouteUsage()) {g=true})
  n16 = OneMinus(None=Power(Base=OneMinus(None=@n12), Exp=LinearInterpolate(A=@n13, B=@n14, Alpha=@n15)))
  n17 = MaterialFunctionCall(
   1=VectorParameter() {parameterName=Color A, defaultValue=<rgba 0 0.03 0.03 0>}.RGB
   2=VectorParameter() {parameterName=Color B, defaultValue=<rgba 0 0.8 0.6 0>}.RGB
   3=VectorParameter() {parameterName=Color C, defaultValue=<rgba 0 1.6 1.0667 0>}.RGB
   4=VectorParameter() {parameterName=Color D, defaultValue=<rgba 0.84 1.2 0.84 0>}.RGB
   A=@n16
  ) {materialFunction=&Lerp_Multiple_Float3}
  n18 = MaterialFunctionCall(
   1=VectorParameter() {parameterName=Variant Color A, defaultValue=<rgba 0 0.0167 0.1 0>}.RGB
   2=VectorParameter() {parameterName=Variant Color B, defaultValue=<rgba 0 0.5 2 0>}.RGB
   3=VectorParameter() {parameterName=Variant Color C, defaultValue=<rgba 0 1.3333 1.6 0>}.RGB
   4=VectorParameter() {parameterName=Variant Color D, defaultValue=<rgba 0.84 1.2 1.14 0>}.RGB
   A=@n16
  ) {materialFunction=&Lerp_Multiple_Float3}
  n19 = DynamicParameter() {defaultValue=<rgba 1 0 0 0>, paramNames [Reactivity Alpha, Reactivity Timeline, Variant, Param4]}
  MaterialFunctionCall(Emissive=Multiply(A=ParticleColor().RGB, B=LinearInterpolate(A=@n17.Lerp 4 Inputs, B=@n18.Lerp 4 Inputs, Alpha=@n19.Variant))) {materialFunction=&MF_FX_EmissiveAdjustments}
  n20 = NamedRerouteUsage()
  n21 = DepthFade(Opacity=@n20, FadeDistance=ScalarParameter() {defaultValue=0.5, parameterName=Depth Fade Distance})
  MaterialFunctionCall("Xb1,PS4"=@n21, "PC-High&&Epic,XSX,PS5"=@n21, "PC-Perf,Android"=@n20, Switch=@n20, "PC-Low(NotPerf)&&Medium"=@n20) {materialFunction=&PlatformSwitch_1D}
  n22 = Constant3Vector() {constant=<rgba 0 -1 0 0>}
  NamedRerouteDeclaration(None=Multiply(A=Transform(None=@n22) {transformSourceType=TRANSFORMSOURCE_Instance}, B=ScalarParameter() {defaultValue=0.5, parameterName=Normal Offset})) {name=Normal Offset, nodeColor=<rgba 0 1 0.5313 1>}
  NamedRerouteDeclaration(None=DynamicParameter() {defaultValue=<rgba 1 0 0 0>, paramNames [Reactivity Alpha, Reactivity Timeline, Variant, Param4]}.Reactivity Alpha) {name=Reactivity Alpha, nodeColor=<rgba 0.5234 0 1 1>}
  NamedRerouteDeclaration(None=Multiply(A=Time(), B=ScalarParameter() {defaultValue=1, parameterName=Time Scale})) {name=Time Scaled, nodeColor=<rgba 0.7578 0 1 1>}
  n23 = NamedRerouteUsage()
  n24 = Divide(
   A=DynamicParameter() {defaultValue=<rgba 1 0 0 0>, paramNames [Reactivity Alpha, Reactivity Timeline, Variant, Param4]}.Reactivity Timeline
   B=ScalarParameter() {defaultValue=0.266, parameterName=Pulse Duration}
  )
  n25 = MaterialFunctionCall!(
   Gradient=MaterialFunctionCall!(
    Gradient=ComponentMask(None=@n23) {g=true}
    Mask Position=Saturate(None=@n24)
    Mask Size=ScalarParameter() {defaultValue=1.5, parameterName=Pulse Mask Size}
   ) {materialFunction=/Game/Effects/Fort_Effects/MaterialFunctions/MF_GradientMask}.Linear
   Center Point=ScalarParameter() {defaultValue=0.9, parameterName=Pulse Gradient Reflected Center Point}
  ) {materialFunction=/Game/Effects/Fort_Effects/MaterialFunctions/MF_GradientReflected}
  n26 = ComponentMask(None=Add(A=TextureCoordinate(), B=NamedRerouteUsage())) {r=true}
  n27 = OneMinus(None=Abs(None=ConstantBiasScale(None=@n26) {bias=-0.5, scale=2}))
  NamedRerouteDeclaration(None=ComponentMask(None=MaterialFunctionCall(Blend=@n25.Result, Base=@n27) {materialFunction=&Blend_Screen}.Result) {r=true}) {name=Center Gradient, nodeColor=<rgba 0 1 0.5078 1>}
  n28 = ScalarParameter() {defaultValue=0.1, parameterName=Distortion UV Scale U}
  n29 = ScalarParameter() {defaultValue=0.3, parameterName=Distortion UV Scale V}
  n30 = MaterialFunctionCall(
   UV=Multiply(A=NamedRerouteUsage(), B=AppendVector(A=@n28, B=@n29))
   Rotation=ScalarParameter() {defaultValue=0.75, parameterName=Distortion Panner Rotation}
   Speed=ScalarParameter() {defaultValue=0.1, parameterName=Distortion Panner Speed}
   Time=NamedRerouteUsage()
  ) {materialFunction=&MF_3UVPanner}
  n31 = TextureObject() {samplerType=SAMPLERTYPE_Normal, texture=/Game/Packages/Fortnite_UniqueMaterials_1/Textures/T_Water_N}
  n32 = Add(
   A=Add(
    A=TextureSample(UVs=@n30.UV1, Tex=@n31) {samplerType=SAMPLERTYPE_Normal}.RGB
    B=TextureSample(UVs=@n30.UV2, Tex=@n31) {samplerType=SAMPLERTYPE_Normal}.RGB
   )
   B=TextureSample(UVs=@n30.UV3, Tex=@n31) {samplerType=SAMPLERTYPE_Normal}.RGB
  )
  n33 = ComponentMask(None=TextureCoordinate()) {r=true}
  n34 = OneMinus(None=Abs(None=ConstantBiasScale(None=@n33) {bias=-0.5, scale=2}))
  n35 = ScalarParameter() {defaultValue=1, parameterName=Distortion Falloff Exponent}
  n36 = ScalarParameter() {defaultValue=0.333, parameterName=Distortion Intensity U}
  n37 = ScalarParameter() {defaultValue=0.333, parameterName=Distortion Intensity V}
  n38 = Multiply(
   A=ComponentMask(None=Divide(A=@n32) {constB=3}) {g=true, r=true}
   B=Multiply(A=Power(Base=@n34, Exp=@n35) {constExponent=4}, B=AppendVector(A=@n36, B=@n37))
  )
  n39 = Constant2Vector()
  NamedRerouteDeclaration(None=MaterialFunctionCall!("Xb1,PS4"=@n38, "PC-High&&Epic,XSX,PS5"=@n38, "PC-Perf,Android"=@n39, Switch=@n39, "PC-Low(NotPerf)&&Medium"=@n39) {materialFunction=/Game/Effects/Niagara/MasterMaterials/Functions/Common/PlatformSwitch_2D}.Result) {name=Distortion, nodeColor=<rgba 1 0 0.1641 1>}
  n40 = NamedRerouteUsage()
  n41 = ScalarParameter() {defaultValue=1, parameterName=Lightning UV Scale U}
  n42 = ScalarParameter() {defaultValue=1, parameterName=Lightning UV Scale V}
  n43 = Add(A=Multiply(A=Subtract(A=@n40) {constB=0.5}, B=AppendVector(A=@n41, B=@n42))) {constB=0.5}
  n44 = AppendVector(A=Floor(None=ScalarParameter() {defaultValue=1, parameterName=Lightning Flipbook Frame}), B=Constant())
  n45 = Add(A=Multiply(A=Add(A=@n43, B=@n44), B=Constant2Vector() {g=1, r=0.3333}), B=NamedRerouteUsage())
  n46 = NamedRerouteUsage()
  n47 = AppendVector(A=Constant(), B=ScalarParameter() {defaultValue=-0.666, parameterName=Lightning Panner Speed})
  NamedRerouteDeclaration(None=TextureSample(
    UVs=Panner(Coordinate=@n45, Time=@n46, Speed=@n47) {bFractionalPart=true, speedY=1}
    Tex=TextureObject() {texture=/Game/Accessories/FORT_Capes/Cape_M_MED_BariumDemon/FX/T_FX_BariumDemon_Electricity}
   ).G) {name=Lightning, nodeColor=<rgba 1 0 0.4453 1>}
  n48 = TextureCoordinate()
  n49 = ComponentMask(None=@n48) {r=true}
  n50 = ComponentMask(None=@n48) {g=true}
  n51 = Multiply(
   A=Abs(None=ConstantBiasScale(None=@n49) {bias=-0.5, scale=2})
   B=Multiply(A=ScalarParameter() {defaultValue=0.3, parameterName=UV Center Skew}, B=NamedRerouteUsage())
  )
  NamedRerouteDeclaration(None=AppendVector(A=@n49, B=Add(A=@n50, B=@n51))) {name=UV Skewed, nodeColor=<rgba 0.4531 1 0 1>}
  n52 = TextureCoordinate()
  n53 = Divide(A=OneMinus(None=ComponentMask(None=@n52) {g=true}), B=ScalarParameter() {defaultValue=0.3, parameterName=Opacity Bottom Fade Size})
  n54 = Saturate(None=MaterialFunctionCall(Base=Saturate(None=@n53), Blend=@n7) {materialFunction=&Blend_Overlay}.Result)
  n55 = ScalarParameter() {defaultValue=2, parameterName=Opacity Bottom Fade Exponent}
  n56 = Divide(A=OneMinus(None=ComponentMask(None=@n52) {g=true}), B=ScalarParameter() {defaultValue=0.2, parameterName=Opacity Bottom Fade Clip Size})
  NamedRerouteDeclaration(None=Multiply(
    A=MaterialFunctionCall(In=@n54, Power=@n55) {materialFunction=&SCurve}.Result
    B=Saturate(None=@n56)
   )) {name=Bottom Fade, nodeColor=<rgba 0 1 0.3906 1>}
  n57 = Subtract(A=CameraPositionWS(), B=WorldPosition().XYZ)
  NamedRerouteDeclaration(None=Multiply(A=Normalize(VectorInput=@n57), B=ScalarParameter() {parameterName=Camera Offset})) {name=Camera Offset, nodeColor=<rgba 0.8047 1 0 1>}
  VertexNormalWS()
 ]
}
