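# Translucent unlit, two-sided Niagara mesh-particle material for the mouth flame of the GateHound backpack cosmetic.
# Two scaled, panning wispy noise samples are combined, biased by particle opacity, and drive erosion of a V-gradient for color.
# The same noise erodes an opacity shape built from fresnel, U-edge fade and V gradient.
# Dynamic parameters fade the flame and blend to a variant palette; Color B's alpha doubles as a noise exponent.
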Material /BRCosmetics/Accessories/FORT_Backpacks/Backpack_M_MED_GateHound/FX/Materials/M_Backpack_GateHound_MouthFlame
graph {
 properties {
  bUsedWithNiagaraMeshParticles=true
  blendMode=BLEND_Translucent
  pixelDepthOffsetMode=PDOM_Legacy
  shadingModel=MSM_Unlit
  translucencyPass=MTP_BeforeDOF
  twoSided=true
  usedShadingModels=Unlit
 }
 nodes [
  n1 = VectorParameter() {parameterName=Color A, defaultValue=<rgba 0 0.02 0.0117 0>}
  n2 = VectorParameter() {parameterName=Color B, defaultValue=<rgba 0 1 0.5 0.66>}
  n3 = VectorParameter() {parameterName=UV Scale and Speed, defaultValue=<rgba 1.75 0.5 -0.2 0.66>}
  n4 = Multiply(A=TextureCoordinate(), B=ComponentMask(None=@n3.RGB) {g=true, r=true})
  n5 = NamedRerouteUsage()
  n6 = AppendVector(A=@n3.B, B=@n3.B)
  n7 = VectorParameter() {parameterName=UV Scale and Speed B, defaultValue=<rgba 1.333 0.333 0.3 1>}
  n8 = Multiply(A=TextureCoordinate(), B=ComponentMask(None=@n7.RGB) {g=true, r=true})
  n9 = AppendVector(A=@n7.B, B=@n7.B)
  n10 = Add(
   A=TextureSample(UVs=Panner(Coordinate=@n4, Time=@n5, Speed=@n6)) {texture=/BRCosmetics/Athena/Effects/Particles/Skydiving/NS_Contrail_Chainmail/T_WispyFireNoise03}.R
   B=TextureSample(UVs=Panner(Coordinate=@n8, Time=@n5, Speed=@n9)) {texture=&T_WispyFireNoise03}.R
  )
  n11 = Saturate(None=DynamicParameter() {paramNames [Opacity,Param2,Variant,Param4]}.Opacity)
  n12 = Saturate(None=Subtract(A=Divide(A=@n10), B=OneMinus(None=@n11)))
  n13 = MaterialFunctionCall!(Gradient=ComponentMask(None=TextureCoordinate()) {g=true}, Erosion Amount=OneMinus(None=Power(Base=@n12, Exp=@n2.A))) {materialFunction=/BRCosmetics/Athena/Effects/Particles/Skydiving/NS_Contrail_DarkAndLight/Functions/MF_ErosionNormalized}
  n14 = VectorParameter() {parameterName=Variant Color A, defaultValue=<rgba 0 0.005 0.02 0>}
  n15 = VectorParameter() {parameterName=Variant Color B, defaultValue=<rgba 0 0.5 1 0.66>}
  MaterialFunctionCall(Emissive=LinearInterpolate(
    A=LinearInterpolate(A=@n1.RGB, B=@n2.RGB, Alpha=@n13.Result)
    B=LinearInterpolate(A=@n14.RGB, B=@n15.RGB, Alpha=@n13.Result)
    Alpha=DynamicParameter() {paramNames [Opacity,Param2,Variant,Param4]}.Variant
   )) {materialFunction=&MF_FX_EmissiveAdjustments}
  n16 = ScalarParameter() {defaultValue=0.5, parameterName=Smoke Fade Size}
  n17 = ComponentMask(None=TextureCoordinate()) {g=true}
  n18 = VertexNormalWS()
  n19 = CameraVectorWS()
  n20 = Power(Base=Abs(None=DotProduct(A=@n18, B=@n19)), Exp=ScalarParameter() {defaultValue=2, parameterName=Fresnel Exponent})
  n21 = ConstantBiasScale(None=ComponentMask(None=TextureCoordinate()) {r=true}) {bias=-0.5, scale=2}
  n22 = OneMinus(None=Power(Base=Abs(None=@n21), Exp=ScalarParameter() {defaultValue=1, parameterName=Edge Fade Exponent}))
  n23 = MaterialFunctionCall!(
   Blend=Power(Base=OneMinus(None=@n17), Exp=ScalarParameter() {defaultValue=1, parameterName=Erosion Exponent})
   Base=OneMinus(None=Multiply(A=@n20, B=@n22))
  ) {materialFunction=/Engine/Functions/Engine_MaterialFunctions03/Blends/Blend_Screen}
  Multiply(
   A=Saturate(None=Divide(A=@n11, B=@n16) {constB=0.5})
   B=OneMinus(None=MaterialFunctionCall(Gradient=@n23.Result, Erosion Amount=@n12) {materialFunction=&MF_ErosionNormalized}.Result)
  )
  NamedRerouteDeclaration(None=Multiply(A=Time(), B=ScalarParameter() {defaultValue=0.6, parameterName=Time Scale})) {name=Time Scaled, nodeColor=<rgba 1 0.7266 0 1>}
 ]
}
