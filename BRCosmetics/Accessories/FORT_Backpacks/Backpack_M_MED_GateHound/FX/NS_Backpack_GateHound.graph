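# Niagara system for the GateHound backpack cosmetic, adding gate and eye glows, mouth flames and embers that react to kill count changes.
# System-level reactivity and timeline values blend idle to reactive colors and drive flame and ember spawning, with time-of-day and oscillation scaling brightness.
# MainElements feeds many sprite, light and mesh renderers through emitter parameters from a burst particle.
# GateEyeGlows pins two glows to skeletal mesh bones or sockets.
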NiagaraSystem /BRCosmetics/Accessories/FORT_Backpacks/Backpack_M_MED_GateHound/FX/NS_Backpack_GateHound
dependencies {
 usedDataInterfaces [
  &NiagaraDataInterfaceCurve
  &NiagaraDataInterfaceArrayFloat3
  &NiagaraDataInterfaceSpriteRendererInfo
  &NiagaraDataInterfaceSkeletalMesh
  &NiagaraDataInterfaceArrayColor
 ]
 usedDynamicInputs [
  &SelectLinearColorFromArray
  &MakeIntFromFloat
  &SetBoolByIntComparison
  &Divide_Float
  &Frac_Float
  &Waveform
  &SetBoolByFloatComparison
  &FN_FrontEndFloatModifier
  &FNTimeOfDayScalar
 ]
 usedModules [
  &FN_VariantManager
  &self:SystemSpawnScript.SystemScriptSource.SystemScriptGraph.NiagaraNodeAssignment_1.SetVariables_#1
  &SystemState
  &self:SystemSpawnScript.SystemScriptSource.SystemScriptGraph.NiagaraNodeAssignment_3.SetVariables_#2
  &self:SystemSpawnScript.SystemScriptSource.SystemScriptGraph.NiagaraNodeAssignment_0.SetVariables_#3
  &Timeline
  &self:SystemSpawnScript.SystemScriptSource.SystemScriptGraph.NiagaraNodeAssignment_4.SetVariables_#4
  &self:SystemSpawnScript.SystemScriptSource.SystemScriptGraph.NiagaraNodeAssignment_2.SetVariables_#5
 ]
}
emitters {
 Embers {
  emitterSpawnScript [/SetVariables(Emitter.GlowSpriteSize:Vector2f=<4 6>, Emitter.SparkSpriteSize:Vector2f=<0.666 1.333>) {isSetParameters=true}]
  emitterUpdateScript [
   &EmitterState(
    Inactive Response=::NewEnumerator1
    Life Cycle Mode=::NewEnumerator1
    Loop Behavior=::NewEnumerator1
    Loop Delay=2.9
    Loop Duration Mode=::NewEnumerator1
    UseLoopDelay=false
   )
   &SpawnPerFrame(Spawn:NiagaraBool=$System.KillCountChangedThisFrame, Spawn Count=11, Use Spawn Probability=false)
   &SpawnRate(
    Spawn Probability=&MakeCustomFloatFromBool(Bool=$System.ReactivityActive, True Float=1, False Float=0)
    SpawnRate=3
    Use Spawn Probability=true
   )
  ]
  particleSpawnScript [
   &InitializeParticle(
    Color Mode=::NewEnumerator0
    Lifetime Max=1.6
    Lifetime Min=0.5
    Lifetime Mode=::NewEnumerator1
    Position Offset=<0 -5 0>
    UsePositionOffset=true
    Write Material Random=true
   )
   &AddVelocity(
    Cone Angle=0
    Cone Axis=&MakeVector(X=&UniformAOrBFloat(A=-1), Y=-0.4, Z=0)
    Module.UseConeFalloff=false
    Module.UseVelDistribution=false
    Override Seed=false
    Velocity Mode=::NewEnumerator2
    Velocity Speed=&RandomRangeFloat(Minimum=9, Maximum=16)
   )
   &ShapeLocation(
    Box Size=<0 3 0>
    MaskSpawnGroup=false
    Offset=<0 -3.5 -2.5>
    Offset Mode=::NewEnumerator0
    Override Seed=false
    Rotation Mode=::NewEnumerator2
    Shape Primitive=::NewEnumerator2
    Surface Only=false
    UseOwnerScale=true
    Yaw / Pitch / Roll=<0 0 -0.125>
   )
  ]
  particleUpdateScript [
   &ParticleState()
   &Drag(Drag=2, Use Linear Drag=true, Use Rotational Drag=true)
   &AccelerationForce(
    Acceleration=&Multiply_VectorByFloat(Vector=<0 0 22>, Float=$Particles.NormalizedAge)
    Coordinate Space=::Local
   )
   &SolveForcesAndVelocity(Clamp Velocity=false, Limit Acceleration=false)
   /SetVariables(
    Particles.GlowSpriteColor:LinearColor=&MultiplyLinearColor(
     Linear Color=$System.EyeGlowColorIdle
     Scale Factor=&MakeFloat4FromVectorAndFloat(
      "Vector (RGB)"=<3 3 3>
      "Float (Alpha)"=&FloatFromCurve(FloatCurve=DataInterface() {propertyValues {Curve {keys [<at 0 1>,<at 1 0>]}}})
     )
    )
    Particles.SparkSpriteColor:LinearColor=&MultiplyLinearColor(
     Linear Color=$System.EyeGlowColorReactive
     Scale Factor=&MakeFloat4FromVectorAndFloat(
      "Vector (RGB)"=<9 9 6>
      "Float (Alpha)"=&FloatFromCurve(FloatCurve=DataInterface() {propertyValues {Curve {keys [<at 0 1>,<at 1 0>]}}})
     )
    )
   ) {isSetParameters=true}
  ]
  rendererClasses [&NiagaraSpriteRendererProperties]
  renderers [
   {rendererClass=&NiagaraSpriteRendererProperties}
   {rendererClass=&NiagaraSpriteRendererProperties, rendererIndex=1}
  ]
 }
 Flames {
  emitterUpdateScript [
   &EmitterState(Life Cycle Mode=::NewEnumerator1, Loop Behavior=::NewEnumerator1, Loop Duration Mode=::NewEnumerator1, UseLoopDelay=false)
   &SpawnRate(
    SpawnRate=&FloatFromCurve(
     FloatCurve=DataInterface() {propertyValues {Curve {keys [
         {arriveTangent=9.0725, time=0, value=0}
         {arriveTangent=-2.5249, leaveTangent=-2.5249, time=0.1408, value=0.9847}
         {arriveTangent=-0.0405, leaveTangent=-0.0405, time=1, value=0}
        ]}}}
     CurveIndex=$System.Timeline.Time
     Scale Curve=16
    )
    Use Spawn Probability=false
   )
  ]
  particleSpawnScript [
   &InitializeParticle(
    Color=$System.FlameSpriteColor
    Lifetime Max=2.1
    Lifetime Min=1.1
    Lifetime Mode=::NewEnumerator1
    Position Offset=<0 -4.5 -1.1>
    Sprite Rotation Mode=::NewEnumerator2
    Sprite Size Mode=::NewEnumerator3
    Uniform Sprite Size=35
    UsePositionOffset=true
    Write Material Random=true
   )
   &AddVelocity(
    Cone Angle=11
    Cone Axis=<0 -1 -1>
    Module.UseConeFalloff=false
    Module.UseVelDistribution=false
    Override Seed=false
    Velocity Mode=::NewEnumerator2
    Velocity Speed=&RandomRangeFloat(Minimum=9, Maximum=16)
   )
  ]
  particleUpdateScript [
   &CameraOffset(Camera Offset Amount=&FloatFromCurve(
     FloatCurve=DataInterface() {propertyValues {Curve {keys [{tangentMode=RCTM_User, time=0, value=0}, {tangentMode=RCTM_User, time=1, value=1}]}}}
     Scale Curve=25
    ))
   &ParticleState()
   &Drag(Drag=0, Use Linear Drag=true, Use Rotational Drag=true)
   &AccelerationForce(Acceleration=<0 0 16>)
   &SolveForcesAndVelocity(Clamp Velocity=false, Limit Acceleration=false)
   &SubUVAnimation(
    End Frame Range Override=15
    "Sprite Renderer "=DataInterface() {propertyValues {SpriteRenderer=&self:Flames_0.NiagaraSpriteRendererProperties_0}}
    UseEndFrame=false
    UseStartFrame=false
   )
   &ScaleColor(
    Scale Alpha=&Multiply_Float(
     A=&FloatFromCurve(FloatCurve=DataInterface() {propertyValues {Curve {keys [
          {tangentMode=RCTM_User, time=0, value=2}
          {arriveTangent=-0.0405, interpMode=RCIM_Cubic, leaveTangent=-0.0405, tangentMode=RCTM_User, time=0.7009, value=1}
          {arriveTangent=-8.0089, interpMode=RCIM_Cubic, leaveTangent=-8.0089, tangentMode=RCTM_User, time=1, value=0}
         ]}}})
     B=1
    )
    Scale RGB=&VectorFromFloat(Value=&FloatFromCurve(FloatCurve=DataInterface() {propertyValues {Curve {keys [
          {interpMode=RCIM_Cubic, leaveTangent=-3.4509, tangentMode=RCTM_Break, time=0, value=1}
          {interpMode=RCIM_Cubic, tangentMode=RCTM_Break, time=0.5, value=0}
         ]}}}))
    ScaleA=true
    ScaleRGB=true
   )
   &ScaleSpriteSize(Uniform Curve Sprite Scale=DataInterface() {propertyValues {Curve {keys [
        {interpMode=RCIM_Cubic, tangentMode=RCTM_Break, time=0, value=0.1}
        {arriveTangent=1.4487, interpMode=RCIM_Cubic, tangentMode=RCTM_Break, time=1, value=1}
       ]}}})
  ]
  rendererClasses [&NiagaraSpriteRendererProperties]
  renderers [{rendererClass=&NiagaraSpriteRendererProperties}]
 }
 GateEyeGlows {
  emitterSpawnScript [/SetVariables(
    Emitter.EyeGlowCameraOffset=6
    Emitter.EyeGlowGateLPositionOffset:NiagaraPosition=<7.25 -4 -8.666>
    Emitter.EyeGlowGateRPositionOffset:NiagaraPosition=<-7.25 -4 -8.666>
   ) {isSetParameters=true, module=SetVariables_#6}]
  emitterUpdateScript [
   &EmitterState(
    Inactive Response=::NewEnumerator1
    Life Cycle Mode=::NewEnumerator1
    Loop Behavior=::NewEnumerator1
    Loop Duration Mode=::NewEnumerator1
    UseLoopDelay=false
   )
   &SpawnBurst_Instantaneous(Spawn Count=2, Use Spawn Probability=false, UseLoopCountLimit=true)
   /SetVariables(Emitter.EyeGlowColor:LinearColor=&MultiplyLinearColor(
     Linear Color=&LerpLinearColors(StartColor=$System.EyeGlowColorIdle, EndColor=$System.EyeGlowColorReactive, LerpFactor=$System.ReactivityAlpha)
     Scale Factor=&MakeVector4_FromFloat(Float=$System.BrightnessOscillation)
    )) {isSetParameters=true, module=SetVariables_#7}
  ]
  particleSpawnScript [
   &InitializeParticle(
    Color=<rgba 0 11 3.6667 0.06>
    Lifetime=5
    Sprite Size Mode=::NewEnumerator3
    Uniform Sprite Size=11
    UsePositionOffset=false
    Write Material Random=true
   )
   &PartitionParticles()
  ]
  particleUpdateScript [
   &ParticleState(Kill Particles When Lifetime Has Elapsed=false)
   &DynamicMaterialParameters(
    Index 0 Param 1=$System.ReactivityAlpha
    Module.Index0Param0Write=true
    Module.Index0Param1Write=false
    Module.Index0Param2Write=false
    Module.Index0Param3Write=false
   )
   &SkeletalMeshLocation(
    Bone / Socket Index=$Particles.PartitionParticles.Partition
    Bone Sampling Mode=::NewEnumerator4
    MaskSpawnGroup=false
    Sampled Position Offset=&SelectVectorFromArray(
     Direct Array Index=$Particles.PartitionParticles.Partition
     Vector Selection Array=DataInterface() {propertyValues {FloatData [<14 2 2>, <14 -2 2>]}}
    )
    Skeletal Mesh=$System.SkeletalMesh
   )
  ]
  rendererClasses [&NiagaraSpriteRendererProperties]
  renderers [{rendererClass=&NiagaraSpriteRendererProperties}]
 }
 MainElements {
  emitterSpawnScript [/SetVariables(
    Emitter.EyeFlareCircularCameraOffset=5
    Emitter.EyeFlareCircularSize:Vector2f=&Vector2DFromFloat(Value=13)
    Emitter.EyeFlareSize:Vector2f=<11 16>
    Emitter.EyeGlowCameraOffset=6
    Emitter.EyeGlowLPosition:NiagaraPosition=<-2.333 -7.5 1.5>
    Emitter.EyeGlowRPosition:NiagaraPosition=<2.333 -7.5 1.5>
    Emitter.GateLightPosition:NiagaraPosition=<0 -7 -9>
    Emitter.GateSoftGlowPosition:NiagaraPosition=<0 -3 -10>
    Emitter.GateSoftGlowSize:Vector2f=&Vector2DFromFloat(Value=19)
    Emitter.GateSpriteAlignment:Vector3f=<0 0 1>
    Emitter.GateSpriteFacing:Vector3f=<0 -1 0>
    Emitter.GateSpritePosition:NiagaraPosition=<0 -1.666 0>
    Emitter.GateSpriteSize:Vector2f=<13 20>
    Emitter.MouthFlamePosition:NiagaraPosition=<0 -9.5 -1.25>
    Emitter.MouthLightPosition:NiagaraPosition=<0 -9.5 -2.5>
   ) {isSetParameters=true, module=SetVariables_#6}]
  emitterUpdateScript [
   &EmitterState(
    Inactive Response=::NewEnumerator1
    Life Cycle Mode=::NewEnumerator1
    Loop Behavior=::NewEnumerator1
    Loop Duration Mode=::NewEnumerator1
    UseLoopDelay=false
   )
   &SpawnBurst_Instantaneous(Use Spawn Probability=false, UseLoopCountLimit=true)
   /SetVariables(
    Emitter.EyeFlareCircularColor:LinearColor=&MultiplyLinearColor(
     Linear Color=$System.EyeGlowColorReactive
     Scale Factor=&MakeFloat4FromVectorAndFloat(
      "Vector (RGB)"=<11 11 11>
      "Float (Alpha)"=&FloatFromCurve(
       FloatCurve=DataInterface() {propertyValues {Curve {keys [
           {interpMode=RCIM_Cubic, leaveTangent=-2.8329, tangentMode=RCTM_Break, time=0, value=1}
           {arriveTangent=-0.0123, interpMode=RCIM_Cubic, leaveTangent=-8.0089, tangentMode=RCTM_Break, time=1, value=0}
          ]}}}
       CurveIndex=&Multiply_Float(A=$System.Timeline.Time, B=$User.ReactivityDuration)
       Scale Curve=&Multiply_Float(A=0.3, B=$System.TimeOfDayBrightnessMultiplier)
      )
     )
    )
    Emitter.EyeFlareCircularDynamic:Vector4f=&MakeVector4(
     X=&FloatFromCurve(
      FloatCurve=DataInterface() {propertyValues {Curve {keys [
          <at 0 1.5>
          {interpMode=RCIM_Cubic, leaveTangent=-2.9741, tangentMode=RCTM_Break, time=0.0666, value=0.8}
          {arriveTangent=-0.0372, interpMode=RCIM_Cubic, tangentMode=RCTM_Break, time=1, value=0}
         ]}}}
      CurveIndex=&Multiply_Float(A=$System.Timeline.Time, B=$User.ReactivityDuration)
     )
     Y=1
     Z=0.8
     W=0
    )
    Emitter.EyeFlareColor:LinearColor=&MultiplyLinearColor(
     Linear Color=$System.EyeGlowColorReactive
     Scale Factor=&MakeFloat4FromVectorAndFloat(
      "Vector (RGB)"=<5 5 5>
      "Float (Alpha)"=&FloatFromCurve(
       FloatCurve=DataInterface() {propertyValues {Curve {keys [{interpMode=RCIM_Cubic, leaveTangent=-3.5062, tangentMode=RCTM_Break, time=0, value=1}, {tangentMode=RCTM_User, time=0.333, value=0}]}}}
       CurveIndex=$System.Timeline.Time
       Scale Curve=&Multiply_Float(A=11, B=$System.TimeOfDayBrightnessMultiplier)
      )
     )
    )
    Emitter.EyeGlowColor:LinearColor=&MultiplyLinearColor(
     Linear Color=&LerpLinearColors(StartColor=$System.EyeGlowColorIdle, EndColor=$System.EyeGlowColorReactive, LerpFactor=$System.ReactivityAlpha)
     Scale Factor=&MakeVector4_FromFloat(Float=&Multiply_Float(A=$System.TimeOfDayBrightnessMultiplier, B=$System.BrightnessOscillation))
    )
    Emitter.GateCapsuleGlowColor:LinearColor=&MultiplyLinearColor(
     Linear Color=&LerpLinearColors(
      StartColor=&MultiplyLinearColor(Linear Color=$System.FlameSpriteColor, Scale Factor=&MakeVector4_FromFloat(Float=1))
      EndColor=$System.FlameSpriteColor
      LerpFactor=$System.ReactivityAlpha
     )
     Scale Factor=&MakeFloat4FromVectorAndFloat(
      "Vector (RGB)"=<1 1 1>
      "Float (Alpha)"=&Multiply_Float(A=$System.TimeOfDayBrightnessMultiplier, B=$System.BrightnessOscillation)
     )
    )
    Emitter.GateCapsuleGlowDynamic:Vector4f=&MakeFloat4FromVectorAndFloat("Vector (RGB)"=$Engine.Owner.SystemZAxis, "Float (Alpha)"=33)
    Emitter.GateLightColor:LinearColor=&MultiplyLinearColor(
     Linear Color=&LerpLinearColors(StartColor=$System.GateLightColorIdle, EndColor=$System.GateLightColorReactive, LerpFactor=$System.ReactivityAlpha)
     Scale Factor=&MakeVector4_FromFloat(Float=$System.BrightnessOscillation)
    )
    Emitter.GateLightRadius:NiagaraFloat=&Lerp_Float(A=12, B=16, Alpha=$System.ReactivityAlpha)
    Emitter.GateSpriteColor:LinearColor=&MakeLinearColorFromVectorAndFloat(
     "Vector (RGB)"=&VectorFromFloat(Value=&Multiply_Float(A=$System.TimeOfDayBrightnessMultiplier, B=$System.BrightnessOscillation))
     "Float (Alpha)"=1
    )
    Emitter.GateSpriteDynamic:Vector4f=&MakeVector4(X=$System.ReactivityAlpha, Y=$System.Timeline.Time, Z=$User.Variant, W=0)
    Emitter.MouthLightColor:LinearColor=&MultiplyLinearColor(
     Linear Color=&MultiplyLinearColor(
      Linear Color=$System.GateLightColorReactive
      Scale Factor=&MakeFloat4FromVectorAndFloat("Vector (RGB)"=&VectorFromFloat(Value=$System.ReactivityAlpha), "Float (Alpha)"=1)
     )
     Scale Factor=&MakeVector4_FromFloat(Float=$System.BrightnessOscillation)
    )
   ) {isSetParameters=true, module=SetVariables_#7}
  ]
  particleSpawnScript [&InitializeParticle(
    Color Mode=::NewEnumerator0
    Lifetime=5
    Sprite Size Mode=::NewEnumerator3
    Uniform Sprite Size=11
    UsePositionOffset=false
    Write Material Random=true
   )]
  particleUpdateScript [&ParticleState(Kill Particles When Lifetime Has Elapsed=false)]
  rendererClasses [
   &NiagaraSpriteRendererProperties
   &NiagaraLightRendererProperties
   &NiagaraMeshRendererProperties
  ]
  renderers [
   {rendererClass=&NiagaraSpriteRendererProperties}
   {rendererClass=&NiagaraLightRendererProperties, rendererIndex=1}
   {rendererClass=&NiagaraSpriteRendererProperties, rendererIndex=2}
   {rendererClass=&NiagaraSpriteRendererProperties, rendererIndex=3}
   {rendererClass=&NiagaraSpriteRendererProperties, rendererIndex=4}
   {rendererClass=&NiagaraSpriteRendererProperties, rendererIndex=5}
   {rendererClass=&NiagaraSpriteRendererProperties, rendererIndex=6}
   {rendererClass=&NiagaraSpriteRendererProperties, rendererIndex=7}
   {rendererClass=&NiagaraSpriteRendererProperties, rendererIndex=8}
   {rendererClass=&NiagaraSpriteRendererProperties, rendererIndex=9}
   {rendererClass=&NiagaraLightRendererProperties, rendererIndex=10}
   {rendererClass=&NiagaraMeshRendererProperties, rendererIndex=11}
  ]
 }
}
summary {
 systemName=NS_Backpack_GateHound
 emitters [
  {
   emitterName=MainElements
   rendererClasses [
    &NiagaraSpriteRendererProperties
    &NiagaraLightRendererProperties
    &NiagaraMeshRendererProperties
   ]
  }
  {emitterName=GateEyeGlows, rendererClasses [&NiagaraSpriteRendererProperties]}
  {emitterName=Embers, rendererClasses [&NiagaraSpriteRendererProperties]}
  {emitterName=Flames, rendererClasses [&NiagaraSpriteRendererProperties]}
 ]
 userVariables [
  User.bIsFrontEnd {type=NiagaraBool}
  User.bIsFrontEndPreview {type=NiagaraBool}
  User.bisInVehicle {type=NiagaraBool}
  User.KillCount {type=NiagaraInt32}
  User.ReactivityDuration {type=NiagaraFloat, defaultValue=3.5, description=""}
  User.Variant {type=NiagaraFloat, defaultValue=0, description=""}
 ]
}
