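# Anim Blueprint for the GateHound backpack that poses the pack and reacts to the player's kills.
# The AnimGraph applies fixed mouth and jaw offsets, then shifts and rolls the pack bone from spine yaw computed each update.
# Rising AthenaKills stats, or a looping simulated counter in the front end, set kill state and broadcast KillCountChanged, resetting after a retriggerable delay.
# The computed dyn_alpha does not feed the AnimGraph.
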AnimBlueprint /BRCosmetics/Accessories/FORT_Backpacks/Backpack_M_MED_GateHound/Meshes/M_MED_GateHound_Pack_AnimBP
target_skeleton=&/M_MED_GateHound_Pack_Skeleton
graphs [
 AnimGraph {nodes [
   n1 = "Transform (Modify) Bone - Bone: dyn_back_mouth_l"(ComponentPose=Local To Component(LocalPose=Input Pose()), Alpha=1, Translation="0,2.7,0", Rotation="0,0,0", Scale="1,1,1") {
    rotationMode=BMM_Additive
    rotationSpace=BCS_BoneSpace
    translationMode=BMM_Additive
    translationSpace=BCS_BoneSpace
    alphaBoolBlend {blendOption=Linear}
    alphaScaleBiasClamp {bInterpResult=true, interpSpeedDecreasing=4, interpSpeedIncreasing=4}
    boneToModify {boneName=dyn_back_mouth_l}
    translation {y=2.7}
   }
   n2 = "Transform (Modify) Bone - Bone: dyn_back_mouth_r"(ComponentPose=@n1, Alpha=1, Translation="0,-2.7,0", Rotation="0,0,0", Scale="1,1,1") {
    rotationMode=BMM_Additive
    rotationSpace=BCS_BoneSpace
    translationMode=BMM_Additive
    translationSpace=BCS_BoneSpace
    alphaBoolBlend {blendOption=Linear}
    alphaScaleBiasClamp {bInterpResult=true, interpSpeedDecreasing=4, interpSpeedIncreasing=4}
    boneToModify {boneName=dyn_back_mouth_r}
    translation {y=-2.7}
   }
   n3 = "Transform (Modify) Bone - Bone: dyn_center_mouth_jaw"(ComponentPose=@n2, Alpha=1, Translation="0,0,0", Rotation="-30,0,0", Scale="1,1,1") {
    rotationMode=BMM_Additive
    rotationSpace=BCS_BoneSpace
    translationMode=BMM_Additive
    translationSpace=BCS_BoneSpace
    alphaBoolBlend {blendOption=Linear}
    alphaScaleBiasClamp {bInterpResult=true, interpSpeedDecreasing=4, interpSpeedIncreasing=4}
    boneToModify {boneName=dyn_center_mouth_jaw}
    rotation {pitch=-30}
   }
   n4 = "Transform (Modify) Bone - Bone: pack"(
    ComponentPose=@n3
    Alpha=Get PackAlpha() -> (PackAlpha:double)
    Translation=Get BackpackTranslate() -> (BackpackTranslate:Vector)
    Rotation=Get BackpackRotate() -> (BackpackRotate:Rotator)
   ) {
    rotationMode=BMM_Additive
    rotationSpace=BCS_BoneSpace
    translationMode=BMM_Additive
    translationSpace=BCS_BoneSpace
    alphaBoolBlend {blendOption=Linear}
    boneToModify {boneName=pack}
   }
   Output Pose(Result=Component To Local(ComponentPose=@n4)) {layerGroup=None}
  ]}
 EventGraph {nodes [
   n1 = TriggerFrontEndReactivity() {is=title K2Node_CustomEvent}
   > n2 = Increment Int(Value=Get FrontEndReactivityKillCount() -> (FrontEndReactivityKillCount:int=0))
   n3 = WatchCosmeticStat!(
    execute:Exec=@n11.then
    Pawn:Fort Player Pawn Object Reference=Get pawn() -> (pawn:Fort Player Pawn Object Reference)
    StatTag:Fort Stat Manager Tag Structure=(TagName="StatManager.AthenaKills")
    DefaultValue:int=0
    bBroadcastInitialValue:bool=false
    InitialValueBroadcastDelay:float=-1
    ChangeValueBroadcastDelay:float=-1
    bConsumeWaitingNotifications:bool=true
   ) -> (then:Exec, OnStatChanged:Exec, StatName:Name, StatValue:int) {is=title K2Node_AsyncAction}
   n4 = SelectInt(A=@n2, B=@n3.StatValue, bPickA=Get bShouldDisplayFrontendReactivity() -> (bShouldDisplayFrontendReactivity:bool))
   n5 = "To Float (Integer)"(InInt=@n4)
   n6 = "float > float"(A=@n5, B=Get KillCount() -> (KillCount:double)) {is=title K2Node_CallFunction}
   Branch(execute=[@n3.OnStatChanged, @n2."  "], Condition=@n6)
   > Set Active_Kill(Active_Kill:bool=@n6) -> (Output_Get:bool)
   > Set KillCount(KillCount:double=@n5) -> (Output_Get:double)
   > Set dyn_alpha(dyn_alpha:double="Clamp (Float)"(Value=Get KillCount(), Min=0, Max=Get max_stage() -> (max_stage:double))) -> (Output_Get:double)
   > Branch(Condition=Get Active_Kill() -> (Active_Kill:bool))
   > Call Kill Count Changed!(execute:Exec, self:M MED Gate Hound Pack Anim BP Object Reference, KillCount:int=@n4) -> (then:Exec) {is=title K2Node_CallDelegate}
   > RetriggerableDelay(Duration=3)
   > Set dyn_alpha(dyn_alpha=0)
   > Set KillCount(KillCount=0)
   > Set Active_Kill(Active_Kill=false)
   Event BlueprintUpdateAnimation()
   n7 = GetOwnerSkeletalMeshComponent()
   n8 = BreakRotator(InRot=BreakTransform(InTransform=GetDeltaTransformFromRefPose(self=@n7, BoneName=spine_05, BaseName=spine_01)).Rotation)
   > Set BackpackRotate(BackpackRotate:Rotator=MakeRotator(Roll=MapRangeClamped(Value=@n8.Yaw, InRangeA=0, InRangeB=-30, OutRangeB=25))) -> (Output_Get:Rotator)
   > Set BackpackTranslate(BackpackTranslate:Vector=MakeVector(Y=MapRangeClamped(Value=@n8.Yaw, InRangeA=0, InRangeB=-30, OutRangeA=1, OutRangeB=-7))) -> (Output_Get:Vector)
   > Set PackAlpha(PackAlpha:double=1) -> (Output_Get:double)
   Event BlueprintInitializeAnimation()
   > n9 = Set Skel_Mesh(Skel_Mesh=GetOwnerSkeletalMeshComponent())
   n10 = Cast To FortPlayerPawn(execute=[@n12.else, @n9.then], Object=TryGetFortPawn())
   > n11 = Set Pawn(Pawn:Fort Player Pawn Object Reference=@n10) -> (Output_Get:Fort Player Pawn Object Reference)
   Event OnCharacterCustomizationLoaded()
   > n12 = Branch(Condition=Get bShouldDisplayFrontendReactivity())
   > Delay(Duration=1)
   > Set Timer by Event(Delegate=@n1, Time=Get FrontEndReactivityInterval() -> (FrontEndReactivityInterval:float), bLooping=true)
   > TriggerFrontEndReactivity() {is=title K2Node_CallFunction}
   Branch(Condition=Get bIsJumping())
   > Set PackAlpha(PackAlpha=0.15)
   TryGetPawnOwner()
  ]}
 KillCountChanged {nodes [KillCountChanged!() -> (then, KillCount:int) {is=title K2Node_FunctionEntry}]}
]
variables [
 Active_Kill {type=boolean}
 BackpackRotate {type=Rotator}
 BackpackTranslate {type=Vector, default=<0 1 0>}
 current_stage {type=number}
 Did_Kill {type=boolean}
 dyn_alpha {type=number}
 FrontEndReactivityInterval {type=number, default=8}
 FrontEndReactivityKillCount {type=integer}
 KillCount {type=number}
 KillCountChanged {default=()}
 Last_KillCount {type=number}
 max_stage {type=number, default=1}
 PackAlpha {type=number, default=1}
 Pawn {type=FortPlayerPawn}
 Skel_Mesh {type=SkeletalMeshComponent}
]
